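# Idle Niagara effect for the SkelFlask character body, adding socket-attached glowing eye effects.
# Three emitters split particles per eye: mirrored AuraMain meshes, flickering EyeGlow sprites, and rotating EyeWisps flipbook sprites suppressed when the head is hidden.
# Variant selectors and time-of-day tints set color.
# Per-eye alphas, likely from the system's eyelid-rotation sampler, scale all three, with system fade and time-of-day scalars also dimming aura and glow.
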NiagaraSystem /BRCosmetics/Characters/Player/Male/Medium/Bodies/M_MED_SkelFlask/FX/NS_Character_SkelFlask_Idle
dependencies {
 usedDataInterfaces [
  &NiagaraDataInterfaceArrayFloat3
  &NiagaraDataInterfaceCurve
  &NiagaraDataInterfaceColorCurve
  &NiagaraDataInterfaceSocketReader
 ]
 usedDynamicInputs [
  &FNTimeOfDayScalar
  /Niagara/DynamicInputs/Lerp/SmoothLerpOverTime_Float
  &RemapRange
  &DotProduct
  &ReturnMeshOrientationAxis
 ]
 usedModules [
  &self:SystemSpawnScript.SystemScriptSource.SystemScriptGraph.NiagaraNodeAssignment_5.SetVariables_#1
  &self:SystemSpawnScript.SystemScriptSource.SystemScriptGraph.NiagaraNodeAssignment_8.SetVariables_#2
  &self:SystemSpawnScript.SystemScriptSource.SystemScriptGraph.NiagaraNodeAssignment_9.SetVariables_#3
  &SystemState
  &self:SystemSpawnScript.SystemScriptSource.SystemScriptGraph.NiagaraNodeAssignment_2.SetVariables_#4
  &FN_VariantManager
  &Timeline
  &self:SampleEyelidRotation
  &self:SystemSpawnScript.SystemScriptSource.SystemScriptGraph.NiagaraNodeAssignment_6.SetVariables_#5
  &self:SystemSpawnScript.SystemScriptSource.SystemScriptGraph.NiagaraNodeAssignment_0.SetVariables_#6
 ]
}
emitters {
 AuraMain {
  emitterUpdateScript [
   &EmitterState(
    Inactive Response=::NewEnumerator1
    Life Cycle Mode=::NewEnumerator1
    Loop Behavior=::NewEnumerator1
    Loop Duration Mode=::NewEnumerator1
    UseLoopDelay=false
   )
   &SpawnBurst_Instantaneous(Spawn Count=2, Use Spawn Probability=false, UseLoopCountLimit=false)
   /SetVariables(
    Emitter.BaseColor:LinearColor=&VariantColorSelector(
     Variant Amount=4
     Original Color=&MultiplyLinearColor(
      Linear Color=<rgba 0.5098 0.9248 3 1>
      Scale Factor=&MakeVector4_FromLinearColor(LinearColor=&FNTimeOfDayColor(Night Color=<rgba 0.2789 0.35 0.7 1>, Morning Color=<rgba 0.939 0.825 1 1>, Day Color=<rgba 1 1 1 1>, Evening Color=<rgba 1 1 1 1>))
     )
     Variant 1 Color=&MultiplyLinearColor(
      Linear Color=<rgba 0.975 1.9875 3 1>
      Scale Factor=&MakeVector4_FromLinearColor(LinearColor=&FNTimeOfDayColor(Night Color=<rgba 0.2789 0.35 0.7 1>, Morning Color=<rgba 0.939 0.825 1 1>, Day Color=<rgba 1 1 1 1>, Evening Color=<rgba 1 1 1 1>))
     )
     Variant 2 Color=&MultiplyLinearColor(
      Linear Color=<rgba 3 2.1699 0.5098 1>
      Scale Factor=&MakeVector4_FromLinearColor(LinearColor=&FNTimeOfDayColor(Night Color=<rgba 0.2789 0.35 0.7 1>, Morning Color=<rgba 0.939 0.825 1 1>, Day Color=<rgba 1 1 1 1>, Evening Color=<rgba 1 1 1 1>))
     )
     Variant 3 Color=&MultiplyLinearColor(
      Linear Color=<rgba 1 0.086 0.9238 1>
      Scale Factor=&MakeVector4_FromLinearColor(LinearColor=&FNTimeOfDayColor(Night Color=<rgba 0.4655 0.5051 0.7 1>, Morning Color=<rgba 0.939 0.825 1 1>, Day Color=<rgba 1 1 1 1>, Evening Color=<rgba 1 1 1 1>))
     )
     Variant 4 Color=&MultiplyLinearColor(
      Linear Color=<rgba 0.066 1.8264 3 1>
      Scale Factor=&MakeVector4_FromLinearColor(LinearColor=&FNTimeOfDayColor(Night Color=<rgba 0.2789 0.35 0.7 1>, Morning Color=<rgba 0.8798 0.655 1 1>, Day Color=<rgba 0.8 0.8 0.8 1>, Evening Color=<rgba 1 1 1 1>))
     )
     Variant 5 Color=&MultiplyLinearColor(
      Linear Color=<rgba 0.36 2.8263 3 1>
      Scale Factor=&MakeVector4_FromLinearColor(LinearColor=&FNTimeOfDayColor(Night Color=<rgba 0.2789 0.35 0.7 1>, Morning Color=<rgba 0.8258 0.5 1 1>, Day Color=<rgba 0.8 0.8 0.8 1>, Evening Color=<rgba 1 1 1 1>))
     )
    )
    Emitter.InnerColor:LinearColor=&VariantColorSelector(
     Variant Amount=4
     Original Color=&MultiplyLinearColor(
      Linear Color=<rgba 1.4116 4.2744 10 1>
      Scale Factor=&MakeVector4_FromLinearColor(LinearColor=&FNTimeOfDayColor(Night Color=<rgba 0.2789 0.35 0.7 1>, Morning Color=<rgba 0.939 0.825 1 1>, Day Color=<rgba 1 1 1 1>, Evening Color=<rgba 1 1 1 1>))
     )
     Variant 1 Color=&MultiplyLinearColor(
      Linear Color=<rgba 0.65 1.4292 10 1>
      Scale Factor=&MakeVector4_FromLinearColor(LinearColor=&FNTimeOfDayColor(Night Color=<rgba 0.2789 0.35 0.7 1>, Morning Color=<rgba 0.939 0.825 1 1>, Day Color=<rgba 1 1 1 1>, Evening Color=<rgba 1 1 1 1>))
     )
     Variant 2 Color=&MultiplyLinearColor(
      Linear Color=<rgba 10 5.1333 1.4116 1>
      Scale Factor=&MakeVector4_FromLinearColor(LinearColor=&FNTimeOfDayColor(Night Color=<rgba 0.2789 0.35 0.7 1>, Morning Color=<rgba 0.939 0.825 1 1>, Day Color=<rgba 1 1 1 1>, Evening Color=<rgba 1 1 1 1>))
     )
     Variant 3 Color=&MultiplyLinearColor(
      Linear Color=<rgba 1.2275 0.43 10 1>
      Scale Factor=&MakeVector4_FromLinearColor(LinearColor=&FNTimeOfDayColor(Night Color=<rgba 0.4655 0.5051 0.7 1>, Morning Color=<rgba 0.939 0.825 1 1>, Day Color=<rgba 1 1 1 1>, Evening Color=<rgba 1 1 1 1>))
     )
     Variant 4 Color=&MultiplyLinearColor(
      Linear Color=<rgba 0.6 7.18 10 1>
      Scale Factor=&MakeVector4_FromLinearColor(LinearColor=&FNTimeOfDayColor(Night Color=<rgba 0.2789 0.35 0.7 1>, Morning Color=<rgba 0.8798 0.655 1 1>, Day Color=<rgba 0.8 0.8 0.8 1>, Evening Color=<rgba 1 1 1 1>))
     )
     Variant 5 Color=&MultiplyLinearColor(
      Linear Color=<rgba 10 2 9.0667 1>
      Scale Factor=&MakeVector4_FromLinearColor(LinearColor=&FNTimeOfDayColor(Night Color=<rgba 0.1554 0.2473 0.7 1>, Morning Color=<rgba 0.8258 0.5 1 1>, Day Color=<rgba 0.8 0.8 0.8 1>, Evening Color=<rgba 1 1 1 1>))
     )
    )
   ) {isSetParameters=true}
  ]
  particleSpawnScript [
   &PartitionParticles()
   &InitializeParticle(
    Color=<rgba 1 1 1 0.6>
    Lifetime Max=1.5
    Lifetime Mode=::NewEnumerator1
    Mesh Scale=&MergeHelper_VectorSwitch(0=<1 1 1>, 1=<-1 1 1>, 2=?, 3=?, Index=$Particles.PartitionParticles.Partition, Num=::NewEnumerator0)
    Mesh Scale Mode=::NewEnumerator4
    UsePositionOffset=true
    Write Material Random=true
   )
   &InitialMeshOrientation(
    Mesh Orientation Mode=::NewEnumerator7
    Rotation=&MergeHelper_VectorSwitch(0=<0 0.25 0>, 1=<0 0.25 0>, 2=?, 3=?, Index=$Particles.PartitionParticles.Partition, Num=::NewEnumerator0)
    Use Rotation Vector=true
   )
   &DynamicMaterialParameters(
    Index 0 Param 1=&MergeHelper_FloatSwitch(0=0, 1=1.35, 2=?, 3=?, Index=$Particles.PartitionParticles.Partition, Num=::NewEnumerator0)
    Module.Index0Float0Write=true
    Module.Index0Float1Write=false
    Module.Index0Float2Write=false
    Module.Index0Float3Write=false
   )
   /SetVariables(Particles.DynamicParameterAlt:Vector4f=&MakeVector4(
     X=&MergeHelper_FloatSwitch(0=1.35, 1=0, 2=?, 3=?, Index=$Particles.PartitionParticles.Partition, Num=::NewEnumerator0)
     Y=0
     Z=1
     W=1
    )) {isSetParameters=true}
   &CameraOffset(Camera Offset Amount=3)
  ]
  particleUpdateScript [
   &ParticleState(Kill Particles When Lifetime Has Elapsed=false, Let Infinitely Lived Particles Die When Emitter Deactivates=true)
   &SocketLocation(
    SamplePosOffset=true
    Sampled Position Offset=&SelectVectorFromArray(
     Direct Array Index=$Particles.PartitionParticles.Partition
     Vector Selection Array=DataInterface() {propertyValues {FloatData [<5 8.353 -4.0115>, <5 8.353 4.0115>]}}
    )
    Socket Reader=$System.Socket Reader
    Socket Sampling Mode=::NewEnumerator3
    VelocitySpeedLimit=false
   )
   &ScaleColor(
    Scale Alpha=&Multiply_Float(
     A=&MaskFloat(
      Mask Source=1
      Mask Parameter=$Particles.PartitionParticles.Partition
      Mask=0
      Masked Float=&FloatFromCurve(
       FloatCurve=DataInterface() {propertyValues {Curve {keys [
           {interpMode=RCIM_Cubic, leaveTangentWeight=0.3333, tangentMode=RCTM_User, time=0, value=0}
           {arriveTangentWeight=0.3333, interpMode=RCIM_Cubic, tangentMode=RCTM_User, time=1, value=1}
          ]}}}
       CurveIndex=&OneMinusFloat(Float=$System.Timeline.Time)
      )
      Passthrough Float=1
     )
     B=&Multiply_Float(
      A=&Multiply_Float(A=$System.SystemFade, B=$System.TODScalar)
      B=&MaskFloat(
       Mask Source=1
       Mask Parameter=$Particles.PartitionParticles.Partition
       Mask=0
       Masked Float=$System.LeftEyeAlpha
       Passthrough Float=$System.RightEyeAlpha
      )
     )
    )
    ScaleA=true
    ScaleRGB=true
   )
   &FN_Mesh_Attributes(
    Color=&MakeLinearColorFromVectorAndFloat(
     "Vector (RGB)"=<1 1 1>
     "Float (Alpha)"=&Multiply_Float(
      A=&Multiply_Float(
       A=&Multiply_Float(
        A=$System.TODScalar
        B=&MaskFloat(
         Mask Source=1
         Mask Parameter=$Particles.PartitionParticles.Partition
         Mask=0
         Masked Float=$System.LeftEyeAlpha
         Passthrough Float=$System.RightEyeAlpha
        )
       )
       B=&Multiply_Float(A=0.225, B=$System.SystemFade)
      )
      B=&MaskFloat(
       Mask Source=1
       Mask Parameter=$Particles.PartitionParticles.Partition
       Mask=0
       Masked Float=&FloatFromCurve(
        FloatCurve=DataInterface() {propertyValues {Curve {keys [
            {interpMode=RCIM_Cubic, leaveTangentWeight=0.3333, tangentMode=RCTM_User, time=0, value=0}
            {arriveTangentWeight=0.3333, interpMode=RCIM_Cubic, tangentMode=RCTM_User, time=1, value=1}
           ]}}}
        CurveIndex=&OneMinusFloat(Float=$System.Timeline.Time)
       )
       Passthrough Float=1
      )
     )
    )
    Create Color Attribute=true
    Create Scale Attribute=true
    Scale=&Multiply_VectorByFloat(Vector=$Particles.Scale, Float=1.2)
    Write to Module Namespace=true
   )
  ]
  rendererClasses [&NiagaraMeshRendererProperties]
  renderers [
   {rendererClass=&NiagaraMeshRendererProperties, rendererIndex=0}
   {rendererClass=&NiagaraMeshRendererProperties, rendererIndex=1}
  ]
 }
 EyeGlow {
  emitterUpdateScript [
   &EmitterState(
    Inactive Response=::NewEnumerator1
    Life Cycle Mode=::NewEnumerator1
    Loop Behavior=::NewEnumerator1
    Loop Duration Mode=::NewEnumerator1
    UseLoopDelay=false
   )
   &SpawnBurst_Instantaneous(Spawn Count=2, Use Spawn Probability=false, UseLoopCountLimit=false)
  ]
  particleSpawnScript [
   &PartitionParticles()
   &InitializeParticle(
    Color=&VariantColorSelector(
     Variant Amount=4
     Original Color=&MultiplyLinearColor(
      Linear Color=<rgba 0.5 2 5 0.06>
      Scale Factor=&MakeVector4_FromLinearColor(LinearColor=&FNTimeOfDayColor(Night Color=<rgba 0.42 0.4673 0.7 1>, Morning Color=<rgba 0.939 0.825 1 1>, Day Color=<rgba 1 1 1 1>, Evening Color=<rgba 1 1 1 1>))
     )
     Variant 1 Color=&MultiplyLinearColor(
      Linear Color=<rgba 0.825 2.9125 5 0.06>
      Scale Factor=&MakeVector4_FromLinearColor(LinearColor=&FNTimeOfDayColor(Night Color=<rgba 0.42 0.4673 0.7 1>, Morning Color=<rgba 0.939 0.825 1 1>, Day Color=<rgba 1 1 1 1>, Evening Color=<rgba 1 1 1 1>))
     )
     Variant 2 Color=&MultiplyLinearColor(
      Linear Color=<rgba 3 1.75 0 0.05>
      Scale Factor=&MakeVector4_FromLinearColor(LinearColor=&FNTimeOfDayColor(Night Color=<rgba 0.1195 0.15 0.3 1>, Morning Color=<rgba 0.9216 0.775 1 1>, Day Color=<rgba 1 1 1 1>, Evening Color=<rgba 1 1 1 1>))
     )
     Variant 3 Color=&MultiplyLinearColor(
      Linear Color=<rgba 2.6333 1.5 10 1>
      Scale Factor=&MakeVector4_FromLinearColor(LinearColor=&FNTimeOfDayColor(
        Night Color=<rgba 0.42 0.4673 0.7 1>
        Morning Color=<rgba 0.3756 0.33 0.4 1>
        Day Color=<rgba 0.4 0.4 0.4 1>
        Evening Color=<rgba 0.6 0.6 0.6 1>
       ))
     )
     Variant 4 Color=&MultiplyLinearColor(
      Linear Color=<rgba 0.188 3.047 4 0.06>
      Scale Factor=&MakeVector4_FromLinearColor(LinearColor=&FNTimeOfDayColor(Night Color=<rgba 0.2789 0.35 0.7 1>, Morning Color=<rgba 0.9216 0.775 1 1>, Day Color=<rgba 0.8 0.8 0.8 1>, Evening Color=<rgba 1 1 1 1>))
     )
     Variant 5 Color=&MultiplyLinearColor(
      Linear Color=<rgba 0.375 2.3937 2.5 0.06>
      Scale Factor=&MakeVector4_FromLinearColor(LinearColor=&FNTimeOfDayColor(Night Color=<rgba 0.1195 0.15 0.3 1>, Morning Color=<rgba 1.3824 1.1625 1.5 1>, Day Color=<rgba 0.8 0.8 0.8 1>, Evening Color=<rgba 1 1 1 1>))
     )
    )
    Lifetime=5
    Sprite Size Mode=::NewEnumerator3
    Uniform Sprite Size=8.5
    UsePositionOffset=false
    Write Material Random=true
   )
   &CameraOffset(Camera Offset Amount=4)
  ]
  particleUpdateScript [
   &ParticleState(Kill Particles When Lifetime Has Elapsed=false, Let Infinitely Lived Particles Die When Emitter Deactivates=true)
   &SocketLocation(
    SamplePosOffset=true
    Sampled Position Offset=&MergeHelper_VectorSwitch(0=<4.8264 7.4507 -2.6949>, 1=<4.8264 7.4507 2.6949>, 2=?, 3=?, Index=$Particles.PartitionParticles.Partition, Num=::NewEnumerator0)
    Socket Reader=$System.Socket Reader
    Socket Sampling Mode=::NewEnumerator3
    VelocitySpeedLimit=false
   )
   &ScaleColor(
    Scale Alpha=&Multiply_Float(
     A=&FloatFromCurve(
      FloatCurve=DataInterface() {propertyValues {Curve {
         postInfinityExtrap=RCCE_CycleWithOffset
         preInfinityExtrap=RCCE_CycleWithOffset
         keys [
          {interpMode=RCIM_Cubic, tangentMode=RCTM_User, time=0, value=0.6}
          {arriveTangent=0.0833, interpMode=RCIM_Cubic, leaveTangent=0.0833, time=0.1574, value=0.7281}
          {arriveTangent=-0.3221, interpMode=RCIM_Cubic, leaveTangent=-0.3221, time=0.3, value=0.625}
          {arriveTangent=1.0245, interpMode=RCIM_Cubic, leaveTangent=1.0245, time=0.4, value=0.65}
          {arriveTangent=1.1515, interpMode=RCIM_Cubic, leaveTangent=1.1515, time=0.4386, value=0.767}
          {interpMode=RCIM_Cubic, tangentMode=RCTM_User, time=0.5185, value=0.7864}
          {arriveTangent=-0.469, interpMode=RCIM_Cubic, leaveTangent=-0.469, time=0.6286, value=0.578}
          {arriveTangent=0.0711, interpMode=RCIM_Cubic, leaveTangent=0.0711, time=0.7672, value=0.6698}
          {arriveTangent=-0.5596, interpMode=RCIM_Cubic, leaveTangent=-0.5596, time=0.8293, value=0.5922}
          {arriveTangent=-0.3306, interpMode=RCIM_Cubic, leaveTangent=-0.3306, time=0.9058, value=0.5922}
          {arriveTangent=0.0826, interpMode=RCIM_Cubic, leaveTangent=0.0826, time=0.9467, value=0.5534}
          {tangentMode=RCTM_User, time=1, value=0.6}
         ]
        }}}
      CurveIndex=&Frac_Float(Float=$Engine.Time)
      Scale Curve=&RandomRangeFloat(Minimum=1.8, Maximum=2.2)
     )
     B=$System.SystemFade
    )
    Scale RGB=&VectorFromFloat(Value=&Multiply_Float(
      A=$System.TODScalar
      B=&MaskFloat(
       Mask Source=1
       Mask Parameter=$Particles.PartitionParticles.Partition
       Mask=0
       Masked Float=$System.LeftEyeAlpha
       Passthrough Float=$System.RightEyeAlpha
      )
     ))
    ScaleA=true
    ScaleRGB=true
   )
   &FN_Sprite_Attributes(
    Color=&Multiply_LinearColorByFloat(
     Linear Color=&VariantColorSelector(
      Variant Amount=4
      Original Color=<rgba 0.1507 0.302 0.6 1>
      Variant 1 Color=<rgba 0.1507 0.302 0.6 1>
      Variant 2 Color=<rgba 0.694 0.4547 0.1196 1>
      Variant 3 Color=<rgba 0.2256 0.1507 0.6 1>
      Variant 4 Color=<rgba 0.1507 0.4877 0.6 1>
      Variant 5 Color=<rgba 0.1507 0.5626 0.6 1>
     )
     Float=&Multiply_Float(
      A=&MaskFloat(
       Mask Source=1
       Mask Parameter=$Particles.PartitionParticles.Partition
       Mask=0
       Masked Float=$System.LeftEyeAlpha
       Passthrough Float=$System.RightEyeAlpha
      )
      B=&Multiply_Float(A=$System.SystemFade, B=$System.TODScalar)
     )
    )
    Create Camera Offset Attribute=true
    Create Color Attribute=true
    Create SpriteSize Attribute=true
    SpriteSize=&Vector2DFromFloat(Value=8)
    Write to Module Namespace=true
   )
  ]
  rendererClasses [&NiagaraSpriteRendererProperties]
  renderers [
   {rendererClass=&NiagaraSpriteRendererProperties}
   {rendererClass=&NiagaraSpriteRendererProperties, rendererIndex=1}
  ]
 }
 EyeWisps {
  emitterUpdateScript [
   &EmitterState(Delay First Loop Only=true, Life Cycle Mode=::NewEnumerator1, Loop Delay=0.4, Loop Duration=1, UseLoopDelay=true)
   &SpawnRate(
    Spawn Probability=&MakeCustomFloatFromBool(Bool=$User.bisHeadHidden, True Float=0, False Float=1)
    SpawnRate=20
    Use Spawn Probability=true
   )
  ]
  particleSpawnScript [
   &PartitionParticles()
   &InitializeParticle(
    Lifetime Min=1.6
    Lifetime Mode=::NewEnumerator1
    Sprite Rotation Mode=::NewEnumerator2
    Sprite Size Max=<15 15>
    Sprite Size Min=<9 9>
    Sprite Size Mode=::NewEnumerator2
    Sprite UV Mode=::NewEnumerator3
    UsePositionOffset=false
    Write Material Random=true
   )
   &CameraOffset(Camera Offset Amount=3)
   &DynamicMaterialParameters(
    Index 0 Param 4=&MakeCustomFloatFromBool(Bool=$User.bIsFrontEndPreview, True Float=0, False Float=1)
    Module.Index0Float0Write=false
    Module.Index0Float1Write=false
    Module.Index0Float2Write=false
    Module.Index0Float3Write=true
   )
  ]
  particleUpdateScript [
   &ParticleState()
   &SocketLocation(
    SamplePosOffset=true
    Sampled Position Offset=&MergeHelper_VectorSwitch(0=<4.8264 7.4507 -2.6949>, 1=<4.8264 7.4507 2.6949>, 2=?, 3=?, Index=$Particles.PartitionParticles.Partition, Num=::NewEnumerator0)
    Socket Reader=$System.Socket Reader
    Socket Sampling Mode=::NewEnumerator3
    VelocitySpeedLimit=false
   )
   &SpriteRotationRate(Rotation Rate=&RandomRangeFloat(Minimum=-155, Maximum=155))
   /SetVariables(Particles.SubImageIndex=&FloatFromCurve(FloatCurve=DataInterface() {propertyValues {Curve {keys [<at 0 5>,<at 1 15>]}}})) {isSetParameters=true}
   &ScaleSpriteSize(Scale Factor=&Vector2DFromFloat(Value=&FloatFromCurve(
      FloatCurve=DataInterface() {propertyValues {Curve {keys [
          {interpMode=RCIM_Cubic, leaveTangent=3.0564, tangentMode=RCTM_Break, time=0, value=0}
          {arriveTangent=1.1454, interpMode=RCIM_Cubic, leaveTangent=1.1454, tangentMode=RCTM_User, time=0.05, value=0.8}
          {arriveTangent=-1.5127, interpMode=RCIM_Cubic, tangentMode=RCTM_Break, time=1, value=0.6}
         ]}}}
      Scale Curve=0.8
     )))
   &ScaleColor(
    Scale Mode=::NewEnumerator1
    Scale RGBA=&Multiply_Vector4(
     A=&MakeVector4_FromLinearColor(LinearColor=&VariantColorSelector(
       Variant Amount=4
       Original Color=&MultiplyLinearColor(
        Linear Color=&ColorFromCurve(ColorCurve=DataInterface() {propertyValues {
           AlphaCurve {keys [<at 0 1 RCIM_Cubic>]}
           BlueCurve {keys [<at 0.1048 1 RCIM_Cubic>,<at 0.5156 1 RCIM_Cubic>]}
           GreenCurve {keys [<at 0.1048 0.4667 RCIM_Cubic>,<at 0.5156 0.1667 RCIM_Cubic>]}
           RedCurve {keys [<at 0.1048 0.2 RCIM_Cubic>,<at 0.5156 0 RCIM_Cubic>]}
          }})
        Scale Factor=&MakeVector4_FromLinearColor(LinearColor=&FNTimeOfDayColor(Night Color=<rgba 0.2789 0.35 0.7 1>, Morning Color=<rgba 0.939 0.825 1 1>, Day Color=<rgba 1 1 1 1>, Evening Color=<rgba 1 1 1 1>))
       )
       Variant 1 Color=&MultiplyLinearColor(
        Linear Color=&ColorFromCurve(ColorCurve=DataInterface() {propertyValues {
           AlphaCurve {keys [<at 0 1 RCIM_Cubic>]}
           BlueCurve {keys [<at 0.1048 1 RCIM_Cubic>,<at 0.5156 1 RCIM_Cubic>]}
           GreenCurve {keys [<at 0.1048 0.65 RCIM_Cubic>,<at 0.5156 0.2144 RCIM_Cubic>]}
           RedCurve {keys [<at 0.1048 0.3 RCIM_Cubic>,<at 0.5156 0.143 RCIM_Cubic>]}
          }})
        Scale Factor=&MakeVector4_FromLinearColor(LinearColor=&FNTimeOfDayColor(Night Color=<rgba 0.2789 0.35 0.7 1>, Morning Color=<rgba 0.939 0.825 1 1>, Day Color=<rgba 1 1 1 1>, Evening Color=<rgba 1 1 1 1>))
       )
       Variant 2 Color=&MultiplyLinearColor(
        Linear Color=&ColorFromCurve(ColorCurve=DataInterface() {propertyValues {
           AlphaCurve {keys [<at 0 1 RCIM_Cubic>]}
           BlueCurve {keys [<at 0.1048 0.172 RCIM_Cubic>,<at 0.5156 0 RCIM_Cubic>]}
           GreenCurve {keys [<at 0.1048 0.5308 RCIM_Cubic>,<at 0.5156 0.6667 RCIM_Cubic>]}
           RedCurve {keys [<at 0.1048 1 RCIM_Cubic>,<at 0.5156 1 RCIM_Cubic>]}
          }})
        Scale Factor=&MakeVector4_FromLinearColor(LinearColor=&FNTimeOfDayColor(Night Color=<rgba 0.2789 0.35 0.7 1>, Morning Color=<rgba 0.939 0.825 1 1>, Day Color=<rgba 1 1 1 1>, Evening Color=<rgba 1 1 1 1>))
       )
       Variant 3 Color=&MultiplyLinearColor(
        Linear Color=&ColorFromCurve(ColorCurve=DataInterface() {propertyValues {
           AlphaCurve {keys [<at 0 8 RCIM_Cubic>,<at 1 4>]}
           BlueCurve {keys [<at 0.1048 0.2 RCIM_Cubic>,<at 0.3204 0.6 RCIM_Cubic>]}
           GreenCurve {keys [<at 0.1048 0.02 RCIM_Cubic>,<at 0.3204 0 RCIM_Cubic>]}
           RedCurve {keys [<at 0.1048 0.08 RCIM_Cubic>,<at 0.3204 0.08 RCIM_Cubic>]}
          }})
        Scale Factor=&MakeVector4_FromLinearColor(LinearColor=&FNTimeOfDayColor(Night Color=<rgba 0.2789 0.35 0.7 1>, Morning Color=<rgba 0.939 0.825 1 1>, Day Color=<rgba 0.8 0.8 0.8 1>, Evening Color=<rgba 1 1 1 1>))
       )
       Variant 4 Color=&MultiplyLinearColor(
        Linear Color=&ColorFromCurve(ColorCurve=DataInterface() {propertyValues {
           AlphaCurve {keys [<at 0 1 RCIM_Cubic>]}
           BlueCurve {keys [<at 0.1048 1 RCIM_Cubic>,<at 0.5156 1 RCIM_Cubic>]}
           GreenCurve {keys [<at 0.1048 0.7812 RCIM_Cubic>,<at 0.5156 0.55 RCIM_Cubic>]}
           RedCurve {keys [<at 0.1048 0.125 RCIM_Cubic>,<at 0.5156 0 RCIM_Cubic>]}
          }})
        Scale Factor=&MakeVector4_FromLinearColor(LinearColor=&FNTimeOfDayColor(Night Color=<rgba 0.2789 0.35 0.7 1>, Morning Color=<rgba 1.2386 0.75 1.5 1>, Day Color=<rgba 0.8 0.8 0.8 1>, Evening Color=<rgba 1 1 1 1>))
       )
       Variant 5 Color=&MultiplyLinearColor(
        Linear Color=&ColorFromCurve(ColorCurve=DataInterface() {propertyValues {
           AlphaCurve {keys [<at 0 1 RCIM_Cubic>]}
           BlueCurve {keys [<at 0.1162 1 RCIM_Cubic>,<at 0.3324 0.895 RCIM_Cubic>]}
           GreenCurve {keys [<at 0.1162 0.9167 RCIM_Cubic>,<at 0.3324 0.1 RCIM_Cubic>]}
           RedCurve {keys [<at 0.1162 0 RCIM_Cubic>,<at 0.3324 1 RCIM_Cubic>]}
          }})
        Scale Factor=&MakeVector4_FromLinearColor(LinearColor=&FNTimeOfDayColor(Night Color=<rgba 0.2789 0.35 0.7 1>, Morning Color=<rgba 0.939 0.825 1 1>, Day Color=<rgba 0.8 0.8 0.8 1>, Evening Color=<rgba 1 1 1 1>))
       )
      ))
     B=&MakeVector4FromVectorAndFloat(
      "Vector (RGB)"=<1 1 1>
      "Float (Alpha)"=&FloatFromCurve(
       FloatCurve=DataInterface() {propertyValues {Curve {keys [
           <at 0 0 RCIM_Cubic>
           <at 0.2 1 RCIM_Cubic>
           {arriveTangent=-1.4238, interpMode=RCIM_Cubic, leaveTangent=-1.4238, tangentMode=RCTM_User, time=1, value=0}
          ]}}}
       Scale Curve=&Multiply_Float(
        A=0.125
        B=&MaskFloat(
         Mask Source=1
         Mask Parameter=$Particles.PartitionParticles.Partition
         Mask=0
         Masked Float=$System.LeftEyeAlpha
         Passthrough Float=$System.RightEyeAlpha
        )
       )
      )
     )
    )
    ScaleRGBA=true
   )
   &KillParticles(Kill Particles=$User.bisHeadHidden)
  ]
  rendererClasses [&NiagaraSpriteRendererProperties]
  renderers [{rendererClass=&NiagaraSpriteRendererProperties}]
 }
}
summary {
 systemName=NS_Character_SkelFlask_Idle
 emitters [
  {emitterName=AuraMain, rendererClasses [&NiagaraMeshRendererProperties]}
  {emitterName=EyeWisps, rendererClasses [&NiagaraSpriteRendererProperties]}
  {emitterName=EyeGlow, rendererClasses [&NiagaraSpriteRendererProperties]}
 ]
 userVariables [
  User.bIsAds {type=NiagaraBool}
  User.bIsFrontEnd {type=NiagaraBool}
  User.bIsFrontEndPreview {type=NiagaraBool}
  User.bisHeadHidden {type=NiagaraBool}
  User.Variant {type=NiagaraFloat, defaultValue=0, description=""}
 ]
}
