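# Physics asset for the cloth mesh of the M_MED_SkelFlask male player body cosmetic, setting up bone-driven jacket physics.
# Pelvis, spine_01 and thigh capsules form a locked body frame, with zero-size base_jacket anchor bodies locked to spine_01.
# Seven chains of two or three dyn_jacket capsules hang from those anchors at left, right and back.
# Chain joints allow limited swing and almost no twist, keeping panels from spinning.
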PhysicsAsset /BRCosmetics/Characters/Player/Male/Medium/Bodies/M_MED_SkelFlask/Meshes/M_MED_SkelFlask_cloth_Physics
bodies {
 base_jacket_b_d_01 [{shapeType=Capsule, length=0, radius=0, shapeName=base_jacket_b_d_01_capsule}]
 base_jacket_l_a_01 [{shapeType=Capsule, length=0, radius=0, shapeName=base_jacket_l_a_01_capsule}]
 base_jacket_l_b_01 [{shapeType=Capsule, length=0, radius=0, shapeName=base_jacket_l_b_01_capsule}]
 base_jacket_l_d_01 [{shapeType=Capsule, length=0, radius=0}]
 base_jacket_r_a_01 [{shapeType=Capsule, length=0, radius=0, shapeName=base_jacket_r_a_01_capsule}]
 base_jacket_r_b_01 [{shapeType=Capsule, length=0, radius=0, shapeName=base_jacket_r_b_01_capsule}]
 base_jacket_r_d_01 [{shapeType=Capsule, length=0, radius=0}]
 dyn_jacket_b_d_01 [{shapeType=Capsule, length=9.5, radius=5, shapeName=dyn_jacket_b_d_01_capsule, center=<6.1177 0 4.7942>, rotation=<rot 0 0 90>}]
 dyn_jacket_b_d_02 [{shapeType=Capsule, length=8.5, radius=6, shapeName=dyn_jacket_b_d_02_capsule, center=<6.1177 0 3.7067>, rotation=<rot 0 0 90>}]
 dyn_jacket_l_a_01 [{shapeType=Capsule, length=7, radius=6, center=<6.1177 0.1195 4.8328>, rotation=<rot 0.0004 0.0006 109.7141>}]
 dyn_jacket_l_a_02 [{shapeType=Capsule, length=5.5, radius=7.5, center=<6.6651 -0.2303 6.1352>, rotation=<rot 0.0044 0.0063 100.8735>}]
 dyn_jacket_l_a_03 [{
   shapeType=Capsule
   length=10.5
   radius=7
   shapeName=dyn_jacket_l_a_03_capsule
   center=<8.3237 -3.2216 6.309>
   rotation=<rot 0.0045 0.0064 99.7034>
  }]
 dyn_jacket_l_b_01 [{shapeType=Capsule, length=9.5, radius=5, center=<6.1177 -0.0181 5.9592>, rotation=<rot 0 0.0005 92.2052>}]
 dyn_jacket_l_b_02 [{shapeType=Capsule, length=8, radius=6.5, center=<6.1177 -0.0625 5.9637>, rotation=<rot 0 0.0002 95.4101>}]
 dyn_jacket_l_b_03 [{
   shapeType=Capsule
   length=8
   radius=6.5
   shapeName=dyn_jacket_l_b_03_capsule
   center=<6.1177 -0.0047 5.7057>
   rotation=<rot 0 0.0002 95.4101>
  }]
 dyn_jacket_l_d_01 [{shapeType=Capsule, length=7.5, radius=5.5, center=<6.1177 0 3.81>, rotation=<rot -0.0531 0.0445 91.0914>}]
 dyn_jacket_l_d_02 [{shapeType=Capsule, length=8, radius=5, center=<6.1177 0 2.9489>, rotation=<rot -0.0531 0.0445 91.0914>}]
 dyn_jacket_r_a_01 [{shapeType=Capsule, length=8, radius=6, center=<6.1177 0.7836 5.2926>, rotation=<rot 0.0001 -0.0003 59.7286>}]
 dyn_jacket_r_a_02 [{shapeType=Capsule, length=8, radius=6, center=<6.1177 3.2755 4.4153>, rotation=<rot 0 -0.0001 85.9976>}]
 dyn_jacket_r_a_03 [{
   shapeType=Capsule
   length=7.5
   radius=6
   shapeName=dyn_jacket_r_a_03_capsule
   center=<6.809 0.0206 4.4889>
   rotation=<rot 0.0001 -0.0002 77.9798>
  }]
 dyn_jacket_r_b_01 [{shapeType=Capsule, length=10, radius=5.5, center=<6.1177 0 5.7394>, rotation=<rot 0 -0.0002 88.2236>}]
 dyn_jacket_r_b_02 [{shapeType=Capsule, length=9, radius=6.5, center=<5.8214 0 5.7394>, rotation=<rot 0 -0.0001 93.4228>}]
 dyn_jacket_r_b_03 [{shapeType=Capsule, length=9, radius=6.5, shapeName=dyn_jacket_r_b_03_capsule, center=<6.1177 0 5.7394>, rotation=<rot 0 -0.0002 91.3189>}]
 dyn_jacket_r_d_01 [{shapeType=Capsule, length=7.5, radius=5.5, center=<6.1177 0 3.81>, rotation=<rot -0.0106 -0.0089 91.3802>}]
 dyn_jacket_r_d_02 [{shapeType=Capsule, length=8, radius=5, center=<6.1177 0 2.9489>, rotation=<rot -0.0106 -0.0089 91.3802>}]
 pelvis [{shapeType=Capsule, length=7, radius=12}]
 spine_01 [{shapeType=Capsule, length=6, radius=10, shapeName=spine_01_capsule, center=<7.9875 -0.114 0>, rotation=<rot 0 -90 0>}]
 thigh_l [{shapeType=Capsule, length=41.5, radius=9.5, center=<-21.8081 0.097 -0.042>, rotation=<rot 86.8191 178.2336 -3.6745>}]
 thigh_r [{shapeType=Capsule, length=36, radius=9, center=<21.7429 -0.3597 -0.3404>, rotation=<rot 87.6667 3.7526 1.94>}]
}
constraints [
 {bone1Name=dyn_jacket_l_a_01, bone2Name=base_jacket_l_a_01, swing1LimitDegrees=20, swing2LimitDegrees=20, twistLimitDegrees=1}
 {bone1Name=dyn_jacket_l_a_02, bone2Name=dyn_jacket_l_a_01, swing1LimitDegrees=20, swing2LimitDegrees=40, twistLimitDegrees=1}
 {bone1Name=dyn_jacket_l_b_01, bone2Name=base_jacket_l_b_01, swing1LimitDegrees=20, swing2LimitDegrees=20, twistLimitDegrees=1}
 {bone1Name=dyn_jacket_l_b_02, bone2Name=dyn_jacket_l_b_01, swing1LimitDegrees=15, swing2LimitDegrees=20, twistLimitDegrees=1}
 {
  bone1Name=base_jacket_l_d_01
  bone2Name=spine_01
  swing1LimitDegrees=45
  swing1Motion=Locked
  swing2LimitDegrees=45
  swing2Motion=Locked
  twistLimitDegrees=45
  twistMotion=Locked
 }
 {bone1Name=dyn_jacket_l_d_01, bone2Name=base_jacket_l_d_01, swing1LimitDegrees=10, swing2LimitDegrees=10, twistLimitDegrees=1}
 {bone1Name=dyn_jacket_l_d_02, bone2Name=dyn_jacket_l_d_01, swing1LimitDegrees=10, swing2LimitDegrees=10, twistLimitDegrees=1}
 {bone1Name=dyn_jacket_r_a_01, bone2Name=base_jacket_r_a_01, swing1LimitDegrees=20, swing2LimitDegrees=20, twistLimitDegrees=1}
 {bone1Name=dyn_jacket_r_a_02, bone2Name=dyn_jacket_r_a_01, swing1LimitDegrees=15, swing2LimitDegrees=30, twistLimitDegrees=1}
 {bone1Name=dyn_jacket_r_b_01, bone2Name=base_jacket_r_b_01, swing1LimitDegrees=20, swing2LimitDegrees=20, twistLimitDegrees=1}
 {bone1Name=dyn_jacket_r_b_02, bone2Name=dyn_jacket_r_b_01, swing1LimitDegrees=15, swing2LimitDegrees=20, twistLimitDegrees=1}
 {
  bone1Name=base_jacket_r_d_01
  bone2Name=spine_01
  swing1LimitDegrees=45
  swing1Motion=Locked
  swing2LimitDegrees=45
  swing2Motion=Locked
  twistLimitDegrees=45
  twistMotion=Locked
 }
 {bone1Name=dyn_jacket_r_d_01, bone2Name=base_jacket_r_d_01, swing1LimitDegrees=10, swing2LimitDegrees=10, twistLimitDegrees=1}
 {bone1Name=dyn_jacket_r_d_02, bone2Name=dyn_jacket_r_d_01, swing1LimitDegrees=10, swing2LimitDegrees=10, twistLimitDegrees=1}
 {
  bone1Name=thigh_l
  bone2Name=pelvis
  swing1LimitDegrees=45
  swing1Motion=Locked
  swing2LimitDegrees=45
  swing2Motion=Locked
  twistLimitDegrees=45
  twistMotion=Locked
 }
 {
  bone1Name=thigh_r
  bone2Name=pelvis
  swing1LimitDegrees=45
  swing1Motion=Locked
  swing2LimitDegrees=45
  swing2Motion=Locked
  twistLimitDegrees=45
  twistMotion=Locked
 }
 {
  bone1Name=base_jacket_l_a_01
  bone2Name=spine_01
  swing1LimitDegrees=45
  swing1Motion=Locked
  swing2LimitDegrees=45
  swing2Motion=Locked
  twistLimitDegrees=45
  twistMotion=Locked
 }
 {
  bone1Name=base_jacket_l_b_01
  bone2Name=spine_01
  swing1LimitDegrees=45
  swing1Motion=Locked
  swing2LimitDegrees=45
  swing2Motion=Locked
  twistLimitDegrees=45
  twistMotion=Locked
 }
 {
  bone1Name=base_jacket_r_a_01
  bone2Name=spine_01
  swing1LimitDegrees=45
  swing1Motion=Locked
  swing2LimitDegrees=45
  swing2Motion=Locked
  twistLimitDegrees=45
  twistMotion=Locked
 }
 {
  bone1Name=base_jacket_r_b_01
  bone2Name=spine_01
  swing1LimitDegrees=45
  swing1Motion=Locked
  swing2LimitDegrees=45
  swing2Motion=Locked
  twistLimitDegrees=45
  twistMotion=Locked
 }
 {bone1Name=dyn_jacket_l_a_03, bone2Name=dyn_jacket_l_a_02, swing1LimitDegrees=25, swing2LimitDegrees=40, twistLimitDegrees=1}
 {bone1Name=dyn_jacket_l_b_03, bone2Name=dyn_jacket_l_b_02, swing1LimitDegrees=10, swing2LimitDegrees=20, twistLimitDegrees=1}
 {bone1Name=dyn_jacket_r_a_03, bone2Name=dyn_jacket_r_a_02, swing1LimitDegrees=10, swing2LimitDegrees=30, twistLimitDegrees=1}
 {bone1Name=dyn_jacket_r_b_03, bone2Name=dyn_jacket_r_b_02, swing1LimitDegrees=10, swing2LimitDegrees=20, twistLimitDegrees=1}
 {
  bone1Name=spine_01
  bone2Name=pelvis
  swing1LimitDegrees=45
  swing1Motion=Locked
  swing2LimitDegrees=45
  swing2Motion=Locked
  twistLimitDegrees=45
  twistMotion=Locked
 }
 {
  bone1Name=base_jacket_b_d_01
  bone2Name=spine_01
  swing1LimitDegrees=45
  swing1Motion=Locked
  swing2LimitDegrees=45
  swing2Motion=Locked
  twistLimitDegrees=45
  twistMotion=Locked
 }
 {bone1Name=dyn_jacket_b_d_01, bone2Name=base_jacket_b_d_01, swing1LimitDegrees=10, swing2LimitDegrees=10, twistLimitDegrees=1}
 {bone1Name=dyn_jacket_b_d_02, bone2Name=dyn_jacket_b_d_01, swing1LimitDegrees=10, swing2LimitDegrees=10, twistLimitDegrees=1}
]
