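# Physics asset for the ribbon necklace on the medium male SkelFlask player body, defining capsule bodies and joints for its swinging strands.
# Upper-spine capsules and a zero-size necklace base form a fully locked chain that anchors the dyn_ bones.
# The first dyn bone branches into two two-bone strands, A and B.
# Their joints lock twist and set narrow swing angles, with one swing axis also locked at the B strand's root.
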PhysicsAsset /BRCosmetics/Characters/Player/Male/Medium/Bodies/M_MED_SkelFlask/Meshes/M_MED_SkelFlask_ribbon_necklace_Physics
bodies {
 base_necklace_01 [{shapeType=Capsule, length=0, radius=0, shapeName=base_necklace_01_capsule, rotation=<rot 0 -90 0>}]
 dyn_necklaceA_02 [{shapeType=Capsule, length=1.5, radius=4.5, center=<2.3505 0 3.66>, rotation=<rot 0 0 90>}]
 dyn_necklaceA_03 [{shapeType=Capsule, length=0.5, radius=4.5, center=<1.9529 0 3.66>, rotation=<rot 90 0 0>}]
 dyn_necklaceB_02 [{shapeType=Capsule, length=1.5, radius=4.5, center=<2.3505 0 3.66>, rotation=<rot 0 0 90>}]
 dyn_necklaceB_03 [{shapeType=Capsule, length=0.5, radius=4.5, center=<1.9529 0.7365 3.66>, rotation=<rot 90 0 0>}]
 dyn_necklace_01 [{shapeType=Capsule, length=3.5, radius=4.5, center=<2.5331 0 3.66>, rotation=<rot 86.6243 -0.0017 -90.4355>}]
 spine_04 [{shapeType=Capsule, length=0.5, radius=14.1, shapeName=spine_04_capsule, center=<-1.6579 -0.7069 0>}]
 spine_05 [{shapeType=Capsule, length=8.5, radius=11.5, shapeName=spine_05_capsule, center=<-5.2304 0.5387 0>}]
}
constraints [
 {
  bone1Name=spine_05
  bone2Name=spine_04
  swing1LimitDegrees=45
  swing1Motion=Locked
  swing2LimitDegrees=45
  swing2Motion=Locked
  twistLimitDegrees=45
  twistMotion=Locked
 }
 {
  bone1Name=base_necklace_01
  bone2Name=spine_05
  swing1LimitDegrees=45
  swing1Motion=Locked
  swing2LimitDegrees=45
  swing2Motion=Locked
  twistLimitDegrees=45
  twistMotion=Locked
 }
 {bone1Name=dyn_necklace_01, bone2Name=base_necklace_01, swing1LimitDegrees=15, swing2LimitDegrees=15, twistLimitDegrees=2, twistMotion=Locked}
 {bone1Name=dyn_necklaceA_02, bone2Name=dyn_necklace_01, swing1LimitDegrees=30, swing2LimitDegrees=30, twistLimitDegrees=2, twistMotion=Locked}
 {
  bone1Name=dyn_necklaceA_03
  bone2Name=dyn_necklaceA_02
  swing1LimitDegrees=30
  swing2LimitDegrees=30
  twistLimitDegrees=2
  twistMotion=Locked
 }
 {
  bone1Name=dyn_necklaceB_02
  bone2Name=dyn_necklace_01
  swing1LimitDegrees=20
  swing2LimitDegrees=20
  swing2Motion=Locked
  twistLimitDegrees=2
  twistMotion=Locked
 }
 {
  bone1Name=dyn_necklaceB_03
  bone2Name=dyn_necklaceB_02
  swing1LimitDegrees=15
  swing2LimitDegrees=15
  twistLimitDegrees=2
  twistMotion=Locked
 }
]
