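# Animation Blueprint for the M_MED_SkelFlask player cosmetic body, adding secondary motion to its cap, collar, ribbons and cloth.
# It copies the attached parent's pose into a Control Rig driven by crouch, crouch-moving and DBNO states.
# Chained bone-relative AnimDynamics nodes then feed three RigidBody simulations using cloth, ribbon and necklace-ribbon physics assets.
# EventGraph updates crouch-moving and a deprecated-function GravityOverride, which the AnimGraph does not visibly read.
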AnimBlueprint /BRCosmetics/Characters/Player/Male/Medium/Bodies/M_MED_SkelFlask/Meshes/M_MED_SkelFlask_AnimBP
target_skeleton=&/M_MED_SkelFlask_Skeleton
graphs [
 AnimGraph {nodes [
   n1 = Control Rig!(
    Source=Copy Pose From Mesh() {bUseAttachedParent=true}
    Alpha
    "is Crouching ":double="To Float (Boolean)"(InBool=Get bIsCrouching())
    "is Crouch Moving ":bool="Get Crouch moving "() -> ("Crouch moving ":bool)
    is dbno:double="To Float (Boolean)"(InBool=Get bIsDBNO())
   ) -> (Pose) {controlRigAssetReference {blueprintRigClass=&/M_MED_SkelFlask_CtrlRig.M_MED_SkelFlask_CtrlRig_C}}
   n2 = Anim Dynamics - Bone: dyn_cap_top_01(ComponentPose=Local To Component(LocalPose=@n1), Alpha=1, ExternalForce="0,0,0") {
    angularDampingOverride=0.9999
    angularSpringConstant=1000
    bOverrideAngularDamping=true
    bUseGravityOverride=true
    simulationSpace=BoneRelative
    alphaBoolBlend {blendOption=Linear}
    boundBone {boneName=dyn_cap_top_01}
    componentAppliedLinearAccClamp=<1500 1500 1500>
    componentLinearAccScale=<0.3 0.3 0.3>
    componentLinearVelScale=<0.3 0.3 0.3>
    gravityOverride {x=-100}
    physicsBodyDefinitions [{
      boundBone {boneName=dyn_cap_top_01}
      boxExtents=<15 15 15>
      constraintSetup {angularLimitsMax=<0 5 20>, angularLimitsMin=<0 -5 -20>, angularTarget=<0 0 0>}
      localJointOffset=<4 0 0>
     }]
    relativeSpaceBone {boneName=neck_01}
    simSpaceSettings {maxAngularAcceleration=10, maxAngularVelocity=2, simSpaceAngularAlpha=1}
   }
   n3 = Anim Dynamics - Bone: dyn_cap_l_01(ComponentPose=@n2, Alpha=1, ExternalForce="0,0,0") {
    angularDampingOverride=0.9999
    angularSpringConstant=1000
    bOverrideAngularDamping=true
    bUseGravityOverride=true
    simulationSpace=BoneRelative
    alphaBoolBlend {blendOption=Linear}
    boundBone {boneName=dyn_cap_l_01}
    componentAppliedLinearAccClamp=<1500 1500 1500>
    componentLinearAccScale=<0.3 0.3 0.3>
    componentLinearVelScale=<0.3 0.3 0.3>
    gravityOverride {x=-100}
    physicsBodyDefinitions [{
      boundBone {boneName=dyn_cap_l_01}
      boxExtents=<15 15 15>
      constraintSetup {angularLimitsMax=<0 10 5>, angularLimitsMin=<0 -10 -5>, angularTarget=<0 0 0>}
      localJointOffset=<4 0 0>
     }]
    relativeSpaceBone {boneName=neck_01}
    simSpaceSettings {maxAngularAcceleration=10, maxAngularVelocity=2, simSpaceAngularAlpha=1}
   }
   n4 = Anim Dynamics - Bone: base_cap_r_01(ComponentPose=@n3, Alpha=1, ExternalForce="0,0,0") {
    angularDampingOverride=0.9999
    angularSpringConstant=1000
    bOverrideAngularDamping=true
    bUseGravityOverride=true
    simulationSpace=BoneRelative
    alphaBoolBlend {blendOption=Linear}
    boundBone {boneName=base_cap_r_01}
    componentAppliedLinearAccClamp=<1500 1500 1500>
    componentLinearAccScale=<0.3 0.3 0.3>
    componentLinearVelScale=<0.3 0.3 0.3>
    gravityOverride {x=-100}
    physicsBodyDefinitions [{
      boundBone {boneName=base_cap_r_01}
      boxExtents=<15 15 15>
      constraintSetup {angularLimitsMax=<0 10 5>, angularLimitsMin=<0 -10 -5>, angularTarget=<0 0 0>}
      localJointOffset=<4 0 0>
     }]
    relativeSpaceBone {boneName=neck_01}
    simSpaceSettings {maxAngularAcceleration=10, maxAngularVelocity=2, simSpaceAngularAlpha=1}
   }
   n5 = Anim Dynamics - Bone: dyn_collarA_r_01(ComponentPose=@n4, Alpha=1, ExternalForce="0,0,0") {
    angularDampingOverride=0.9999
    angularSpringConstant=1000
    bOverrideAngularDamping=true
    bUseGravityOverride=true
    simulationSpace=BoneRelative
    worldSpaceMinimumScale=0.5
    alphaBoolBlend {blendOption=Linear}
    boundBone {boneName=dyn_collarA_r_01}
    componentAppliedLinearAccClamp=<1500 1500 1500>
    componentLinearAccScale=<0.3 0.3 0.3>
    componentLinearVelScale=<0.3 0.3 0.3>
    gravityOverride {x=-100}
    physicsBodyDefinitions [{
      boundBone {boneName=dyn_collarA_r_01}
      boxExtents=<15 15 15>
      constraintSetup {angularLimitsMax=<0 20 20>, angularLimitsMin=<0 -20 -20>, angularTarget=<0 0 0>}
      localJointOffset=<5 0 0>
     }]
    relativeSpaceBone {boneName=spine_05}
    simSpaceSettings {maxAngularAcceleration=10, maxAngularVelocity=2, simSpaceAngularAlpha=1}
   }
   n6 = Anim Dynamics - Bone: dyn_collarB_r_01(ComponentPose=@n5, Alpha=1, ExternalForce="0,0,0") {
    angularDampingOverride=0.9999
    angularSpringConstant=1000
    bOverrideAngularDamping=true
    bUseGravityOverride=true
    simulationSpace=BoneRelative
    worldSpaceMinimumScale=0.5
    alphaBoolBlend {blendOption=Linear}
    boundBone {boneName=dyn_collarB_r_01}
    componentAppliedLinearAccClamp=<1500 1500 1500>
    componentLinearAccScale=<0.3 0.3 0.3>
    componentLinearVelScale=<0.3 0.3 0.3>
    gravityOverride {x=-100}
    physicsBodyDefinitions [{
      boundBone {boneName=dyn_collarB_r_01}
      boxExtents=<15 15 15>
      constraintSetup {angularLimitsMax=<0 20 20>, angularLimitsMin=<0 -20 -20>, angularTarget=<0 0 0>}
      localJointOffset=<5 0 0>
     }]
    relativeSpaceBone {boneName=spine_05}
    simSpaceSettings {maxAngularAcceleration=10, maxAngularVelocity=2, simSpaceAngularAlpha=1}
   }
   n7 = Anim Dynamics - Bone: dyn_collarB_l_01(ComponentPose=@n6, Alpha=1, ExternalForce="0,0,0") {
    angularDampingOverride=0.9999
    angularSpringConstant=1000
    bOverrideAngularDamping=true
    bUseGravityOverride=true
    simulationSpace=BoneRelative
    worldSpaceMinimumScale=0.5
    alphaBoolBlend {blendOption=Linear}
    boundBone {boneName=dyn_collarB_l_01}
    componentAppliedLinearAccClamp=<1500 1500 1500>
    componentLinearAccScale=<0.3 0.3 0.3>
    componentLinearVelScale=<0.3 0.3 0.3>
    gravityOverride {x=-100}
    physicsBodyDefinitions [{
      boundBone {boneName=dyn_collarB_l_01}
      boxExtents=<15 15 15>
      constraintSetup {angularLimitsMax=<0 20 20>, angularLimitsMin=<0 -20 -20>, angularTarget=<0 0 0>}
      localJointOffset=<5 0 0>
     }]
    relativeSpaceBone {boneName=spine_05}
    simSpaceSettings {maxAngularAcceleration=10, maxAngularVelocity=2, simSpaceAngularAlpha=1}
   }
   n8 = Anim Dynamics - Bone: dyn_collarA_l_01(ComponentPose=@n7, Alpha=1, ExternalForce="0,0,0") {
    angularDampingOverride=0.9999
    angularSpringConstant=1000
    bOverrideAngularDamping=true
    bUseGravityOverride=true
    simulationSpace=BoneRelative
    worldSpaceMinimumScale=0.5
    alphaBoolBlend {blendOption=Linear}
    boundBone {boneName=dyn_collarA_l_01}
    componentAppliedLinearAccClamp=<1500 1500 1500>
    componentLinearAccScale=<0.3 0.3 0.3>
    componentLinearVelScale=<0.3 0.3 0.3>
    gravityOverride {x=-100}
    physicsBodyDefinitions [{
      boundBone {boneName=dyn_collarA_l_01}
      boxExtents=<15 15 15>
      constraintSetup {angularLimitsMax=<0 20 20>, angularLimitsMin=<0 -20 -20>, angularTarget=<0 0 0>}
      localJointOffset=<5 0 0>
     }]
    relativeSpaceBone {boneName=spine_05}
    simSpaceSettings {maxAngularAcceleration=10, maxAngularVelocity=2, simSpaceAngularAlpha=1}
   }
   n9 = Anim Dynamics - Bone: dyn_thighribbon_01(ComponentPose=@n8, Alpha=1, ExternalForce="0,0,0") {
    angularDampingOverride=0.9999
    bUseGravityOverride=true
    simulationSpace=BoneRelative
    alphaBoolBlend {blendOption=Linear}
    boundBone {boneName=dyn_thighribbon_01}
    componentAppliedLinearAccClamp=<500 500 500>
    componentLinearAccScale=<0.1 0.1 0.1>
    componentLinearVelScale=<0.1 0.1 0.1>
    gravityOverride {x=-100}
    physicsBodyDefinitions [{
      boundBone {boneName=dyn_thighribbon_01}
      boxExtents=<10 10 10>
      constraintSetup {angularLimitsMax=<0 0 1>, angularLimitsMin=<0 -10 -1>, angularTarget=<0 0 0>}
      localJointOffset=<2 0 0>
     }]
    relativeSpaceBone {boneName=thigh_l}
    simSpaceSettings {maxAngularAcceleration=10, maxAngularVelocity=2, simSpaceAngularAlpha=1}
   }
   n10 = Anim Dynamics - Bone: dyn_thighribbon_02(ComponentPose=@n9, Alpha=1, ExternalForce="0,0,0") {
    angularDampingOverride=0.9999
    bUseGravityOverride=true
    simulationSpace=BoneRelative
    alphaBoolBlend {blendOption=Linear}
    boundBone {boneName=dyn_thighribbon_02}
    componentAppliedLinearAccClamp=<500 500 500>
    componentLinearAccScale=<0.1 0.1 0.1>
    componentLinearVelScale=<0.1 0.1 0.1>
    gravityOverride {x=-100}
    physicsBodyDefinitions [{
      boundBone {boneName=dyn_thighribbon_02}
      boxExtents=<10 10 10>
      constraintSetup {angularLimitsMax=<0 0 1>, angularLimitsMin=<0 -10 -1>, angularTarget=<0 0 0>}
      localJointOffset=<2 0 0>
     }]
    relativeSpaceBone {boneName=thigh_l}
    simSpaceSettings {maxAngularAcceleration=10, maxAngularVelocity=2, simSpaceAngularAlpha=1}
   }
   n11 = Anim Dynamics - Bone: dyn_thighribbon_03(ComponentPose=@n10, Alpha=1, ExternalForce="0,0,0") {
    angularDampingOverride=0.9999
    bUseGravityOverride=true
    simulationSpace=BoneRelative
    alphaBoolBlend {blendOption=Linear}
    boundBone {boneName=dyn_thighribbon_03}
    componentAppliedLinearAccClamp=<500 500 500>
    componentLinearAccScale=<0.1 0.1 0.1>
    componentLinearVelScale=<0.1 0.1 0.1>
    gravityOverride {x=-100}
    physicsBodyDefinitions [{
      boundBone {boneName=dyn_thighribbon_03}
      boxExtents=<10 10 10>
      constraintSetup {angularLimitsMax=<0 0 1>, angularLimitsMin=<0 -10 -1>, angularTarget=<0 0 0>}
      localJointOffset=<2 0 0>
     }]
    relativeSpaceBone {boneName=thigh_l}
    simSpaceSettings {maxAngularAcceleration=10, maxAngularVelocity=2, simSpaceAngularAlpha=1}
   }
   n12 = RigidBody!(ComponentPose=@n11, Alpha, OverrideWorldGravity:Vector="0,0,0", ExternalForce) -> (Pose) {
    bClampLinearTranslationLimitToRefPose=true
    bOverrideWorldGravity=true
    overridePhysicsAsset=&/M_MED_SkelFlask_cloth_Physics
    simulationSpace=BaseBoneSpace
    baseBoneRef {boneName=spine_03}
    overrideWorldGravity {z=-980}
    simSpaceSettings {maxAngularAcceleration=100, maxAngularVelocity=5, maxLinearAcceleration=300, maxLinearVelocity=100, velocityScaleZ=0, worldAlpha=1}
   }
   n13 = RigidBody(ComponentPose=@n12, OverrideWorldGravity="0,0,0") {
    bClampLinearTranslationLimitToRefPose=true
    bOverrideWorldGravity=true
    overridePhysicsAsset=&/M_MED_SkelFlask_ribbon_Physics
    simulationSpace=BaseBoneSpace
    baseBoneRef {boneName=thigh_l}
    overrideWorldGravity {z=-980}
    simSpaceSettings {maxAngularAcceleration=100, maxAngularVelocity=5, maxLinearAcceleration=500, maxLinearVelocity=100, velocityScaleZ=0, worldAlpha=1}
   }
   n14 = RigidBody(ComponentPose=@n13, OverrideWorldGravity="0,0,0") {
    bClampLinearTranslationLimitToRefPose=true
    bOverrideWorldGravity=true
    overridePhysicsAsset=&/M_MED_SkelFlask_ribbon_necklace_Physics
    simulationSpace=BaseBoneSpace
    baseBoneRef {boneName=spine_05}
    overrideWorldGravity {z=-980}
    simSpaceSettings {maxAngularAcceleration=100, maxAngularVelocity=5, maxLinearAcceleration=500, maxLinearVelocity=100, velocityScaleZ=0, worldAlpha=1}
   }
   Output Pose(Result=Component To Local(ComponentPose=@n14)) {layerGroup=None}
  ]}
 EventGraph {nodes [
   Event BlueprintUpdateAnimation()
   > n1 = Sequence()
   > "Set Crouch moving "("Crouch moving ":bool=AND Boolean(A=Get bIsCrouching(), B=Get bIsMoving2D())) -> (Output_Get:bool)
   n2 = Gravity Override State Function DEPRECATED(
    execute=@n1.then_1
    CharacterPartAnimInstance=Self-Reference()
    GameTimeInSeconds=GetGameTimeInSeconds()
    GravityOverrideParams=Get Gravity Override Params()
   )
   > Set GravityOverride(GravityOverride=@n2)
  ]}
]
variables [
 "Crouch moving " {type=boolean}
 Gravity Override Params {type=GravityOverrideParamsStruct, default {dBNO {gravityOverride=<-980 980 0>}}}
 GravityOverride {type=Vector}
]
